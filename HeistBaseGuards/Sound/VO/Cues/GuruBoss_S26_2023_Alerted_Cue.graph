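# Sound cue for the GuruBoss heist guard's alerted voice line, from the S26 2023 VO set.
# A Random node picks one of five equally weighted voice waves and is excluded from branch culling.
# It uses the shared NPC alerted VO attenuation and the critical enemy VO sound class, with boosted volume.
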SoundCue /HeistBaseGuards/Sound/VO/Cues/GuruBoss_S26_2023_Alerted_Cue
root=Random_0
nodes [
 {
  id=Random_0
  children [WavePlayer_0,WavePlayer_1,WavePlayer_2,WavePlayer_3,WavePlayer_4]
  properties {bShouldExcludeFromBranchCulling=true, weights [1,1,1,1,1]}
 }
 {id=WavePlayer_0, properties {soundWaveAssetPtr=/HeistBaseGuards/Sound/VO/Waves/GuruBoss/GuruBoss_S26_2023_Alerted_011}}
 {id=WavePlayer_1, properties {soundWaveAssetPtr=/HeistBaseGuards/Sound/VO/Waves/GuruBoss/GuruBoss_S26_2023_Alerted_020}}
 {id=WavePlayer_2, properties {soundWaveAssetPtr=/HeistBaseGuards/Sound/VO/Waves/GuruBoss/GuruBoss_S26_2023_Alerted_032}}
 {id=WavePlayer_3, properties {soundWaveAssetPtr=/HeistBaseGuards/Sound/VO/Waves/GuruBoss/GuruBoss_S26_2023_Alerted_040}}
 {id=WavePlayer_4, properties {soundWaveAssetPtr=/HeistBaseGuards/Sound/VO/Waves/GuruBoss/GuruBoss_S26_2023_Alerted_050}}
]
properties {
 attenuationSettings=/Game/Sounds/Fort_Human_VO_BR/NPCs/Shared/NPC_VO_Alerted_Att
 maxDistance=12500
 soundClassObject=&Fort_VO_Enemy_Critical
 volumeMultiplier=2.7
}
