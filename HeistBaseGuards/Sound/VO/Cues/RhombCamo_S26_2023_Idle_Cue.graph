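# Sound Cue for idle voice lines of the RhombCamo heist base guard NPC.
# A weighted root random, which allows repeats, favors a two-wave group over a three-wave group, so those two lines play more often.
# Each group picks evenly among its waves, and all random nodes are excluded from branch culling.
# Output is volume-boosted and uses shared close-range NPC voice attenuation and the non-critical enemy VO sound class.
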SoundCue /HeistBaseGuards/Sound/VO/Cues/RhombCamo_S26_2023_Idle_Cue
root=Random_1
nodes [
 {id=Random_0, children [WavePlayer_0,WavePlayer_4], properties {bShouldExcludeFromBranchCulling=true, weights [1,1]}}
 {
  id=Random_1
  children [Random_0,Random_2]
  properties {bRandomizeWithoutReplacement=false, bShouldExcludeFromBranchCulling=true, weights [1,0.75]}
 }
 {id=Random_2, children [WavePlayer_5,WavePlayer_6,WavePlayer_7], properties {bShouldExcludeFromBranchCulling=true, weights [1,1,1]}}
 {id=WavePlayer_0, properties {soundWaveAssetPtr=/HeistBaseGuards/Sound/VO/Waves/RhombCamo/RhombCamo_S26_2023_Idle_010}}
 {id=WavePlayer_4, properties {soundWaveAssetPtr=/HeistBaseGuards/Sound/VO/Waves/RhombCamo/RhombCamo_S26_2023_Idle_050}}
 {id=WavePlayer_5, properties {soundWaveAssetPtr=/HeistBaseGuards/Sound/VO/Waves/RhombCamo/RhombCamo_S26_2023_Idle_020}}
 {id=WavePlayer_6, properties {soundWaveAssetPtr=/HeistBaseGuards/Sound/VO/Waves/RhombCamo/RhombCamo_S26_2023_Idle_030}}
 {id=WavePlayer_7, properties {soundWaveAssetPtr=/HeistBaseGuards/Sound/VO/Waves/RhombCamo/RhombCamo_S26_2023_Idle_040}}
]
properties {
 attenuationSettings=/Game/Sounds/Fort_Human_VO_BR/NPCs/Shared/NPC_VO_Close_Att
 maxDistance=8500
 soundClassObject=/Game/Sounds/FortSoundClassesAndModes/Fort_VO_Enemy_NonCritical
 volumeMultiplier=2.825
}
