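# Sound cue for RhombCamo heist base guard voice lines, named for the return from suspicious to idle.
# A root Random node picks among five equally weighted wave players, each holding a different line variant, and is excluded from branch culling.
# It uses the shared NPC alerted VO attenuation and the non-critical enemy VO sound class, with a boosted volume multiplier.
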SoundCue /HeistBaseGuards/Sound/VO/Cues/RhombCamo_S26_2023_SusReturnToIdle_Cue
root=Random_0
nodes [
 {
  id=Random_0
  children [WavePlayer_0,WavePlayer_1,WavePlayer_2,WavePlayer_3,WavePlayer_4]
  properties {bShouldExcludeFromBranchCulling=true, weights [1,1,1,1,1]}
 }
 {id=WavePlayer_0, properties {soundWaveAssetPtr=/HeistBaseGuards/Sound/VO/Waves/RhombCamo/RhombCamo_S26_2023_SusReturnToIdle_010}}
 {id=WavePlayer_1, properties {soundWaveAssetPtr=/HeistBaseGuards/Sound/VO/Waves/RhombCamo/RhombCamo_S26_2023_SusReturnToIdle_011}}
 {id=WavePlayer_2, properties {soundWaveAssetPtr=/HeistBaseGuards/Sound/VO/Waves/RhombCamo/RhombCamo_S26_2023_SusReturnToIdle_030}}
 {id=WavePlayer_3, properties {soundWaveAssetPtr=/HeistBaseGuards/Sound/VO/Waves/RhombCamo/RhombCamo_S26_2023_SusReturnToIdle_040}}
 {id=WavePlayer_4, properties {soundWaveAssetPtr=/HeistBaseGuards/Sound/VO/Waves/RhombCamo/RhombCamo_S26_2023_SusReturnToIdle_051}}
]
properties {
 attenuationSettings=/Game/Sounds/Fort_Human_VO_BR/NPCs/Shared/NPC_VO_Alerted_Att
 maxDistance=12500
 soundClassObject=/Game/Sounds/FortSoundClassesAndModes/Fort_VO_Enemy_NonCritical
 volumeMultiplier=1.8
}
