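# Sound cue for idle voice lines of the female Rhomb guard NPC in HeistBaseGuards.
# A root random node picks, with replacement, between groups of three and four waves, weighted by group size so lines play equally often overall.
# Each group is its own random node with equal weights, excluded from branch culling.
# Output uses shared close-range NPC VO attenuation, the non-critical enemy VO sound class and boosted volume.
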SoundCue /HeistBaseGuards/Sound/VO/Cues/RhombGuard_F_S26_2023_Idle_Cue
root=Random_2
nodes [
 {id=Random_0, children [WavePlayer_0,WavePlayer_1,WavePlayer_2], properties {bShouldExcludeFromBranchCulling=true, weights [1,1,1]}}
 {
  id=Random_1
  children [WavePlayer_3,WavePlayer_4,WavePlayer_5,WavePlayer_6]
  properties {bShouldExcludeFromBranchCulling=true, weights [1,1,1,1]}
 }
 {id=Random_2, children [Random_0,Random_1], properties {bRandomizeWithoutReplacement=false, weights [0.75,1]}}
 {id=WavePlayer_0, properties {soundWaveAssetPtr=/HeistBaseGuards/Sound/VO/Waves/RhombGuards_F/RhombGuard_F_S26_2023_Idle_010}}
 {id=WavePlayer_1, properties {soundWaveAssetPtr=/HeistBaseGuards/Sound/VO/Waves/RhombGuards_F/RhombGuard_F_S26_2023_Idle_020}}
 {id=WavePlayer_2, properties {soundWaveAssetPtr=/HeistBaseGuards/Sound/VO/Waves/RhombGuards_F/RhombGuard_F_S26_2023_Idle_030}}
 {id=WavePlayer_3, properties {soundWaveAssetPtr=/HeistBaseGuards/Sound/VO/Waves/RhombGuards_F/RhombGuard_F_S26_2023_Idle_041}}
 {id=WavePlayer_4, properties {soundWaveAssetPtr=/HeistBaseGuards/Sound/VO/Waves/RhombGuards_F/RhombGuard_F_S26_2023_Idle_050}}
 {id=WavePlayer_5, properties {soundWaveAssetPtr=/HeistBaseGuards/Sound/VO/Waves/RhombGuards_F/RhombGuard_F_S26_2023_Idle_053}}
 {id=WavePlayer_6, properties {soundWaveAssetPtr=/HeistBaseGuards/Sound/VO/Waves/RhombGuards_F/RhombGuard_F_S26_2023_Idle_055}}
]
properties {
 attenuationSettings=/Game/Sounds/Fort_Human_VO_BR/NPCs/Shared/NPC_VO_Close_Att
 maxDistance=8500
 soundClassObject=/Game/Sounds/FortSoundClassesAndModes/Fort_VO_Enemy_NonCritical
 volumeMultiplier=2.825
}
